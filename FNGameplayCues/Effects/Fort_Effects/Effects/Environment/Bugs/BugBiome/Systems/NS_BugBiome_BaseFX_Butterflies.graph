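# Niagara system for bug biome ambient life, with GPU emitters for butterflies, fireflies and wind-blown leaves; the tree butterfly emitter is disabled.
# Ground butterflies spawn on depth-buffer surfaces, filtered by GBuffer hue, then idle until camera-relative point forces set them flying with curl noise.
# Fireflies sample the landscape and spawn by time of day.
# Wind gust and time-of-day NPC parameters modulate spawning, forces and material animation.
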NiagaraSystem /FNGameplayCues/Effects/Fort_Effects/Effects/Environment/Bugs/BugBiome/Systems/NS_BugBiome_BaseFX_Butterflies
dependencies {
 usedDataInterfaces [
  &NiagaraDataInterfaceCurve
  &NiagaraDataInterfaceColorCurve
  &NiagaraDataInterfaceGBuffer
  &NiagaraDataInterfaceCollisionQuery
  &NiagaraDataInterfaceCamera
  &NiagaraDataInterfaceArrayFloat
 ]
 usedDynamicInputs [&SimulationPosition]
 usedModules [&SystemState]
}
emitters {
 Butterflies_InTrees {
  bEnabled=false
  simTarget=GPUComputeSim
  emitterUpdateScript [
   &EmitterState()
   &SpawnBurst_Instantaneous(
    Spawn Count=&RandomRangeInt(Minimum=40, Maximum=200)
    Use Spawn Probability=false
    UseLoopCountLimit=false
   )
   &SpawnRate(
    SpawnRate=&Multiply_Float(
     A=222
     B=&FloatFromCurve(
      FloatCurve=DataInterface() {propertyValues {Curve {
         defaultValue=1
         keys [
          {interpMode=RCIM_Cubic, tangentMode=RCTM_User, time=0, value=0.1}
          {arriveTangent=-0.0252, interpMode=RCIM_Cubic, leaveTangent=-0.0252, tangentMode=RCTM_User, time=0.3, value=1}
          {interpMode=RCIM_Cubic, tangentMode=RCTM_User, time=0.7, value=1}
          {tangentMode=RCTM_User, time=1, value=0.1}
         ]
        }}}
      CurveIndex=Linked() {linkedVariable {name=NPC.FortniteNPC.TimeOfDay_0-1, type=NiagaraFloat}}
     )
    )
    Use Spawn Probability=false
   )
  ]
  particleSpawnScript [
   &InitializeParticle(
    Color=&ColorFromCurve(
     ColorCurve=DataInterface() {propertyValues {
       AlphaCurve {keys [<at 0 1 RCIM_Constant>,<at 1 1 RCIM_Constant>]}
       BlueCurve {keys [<at 0 0.5881 RCIM_Constant>,<at 0.0989 1 RCIM_Constant>,<at 0.1984 0.0729 RCIM_Constant>,<at 0.3016 0.3289 RCIM_Constant>,<at 0.3996 1 RCIM_Constant>,<at 0.5005 1 RCIM_Constant>,<at 0.5977 0.5532 RCIM_Constant>,<at 0.6957 0.3262 RCIM_Constant>,<at 0.7982 0 RCIM_Constant>,<at 0.8984 0 RCIM_Constant>,<at 1 1 RCIM_Constant>]}
       GreenCurve {keys [<at 0 1 RCIM_Constant>,<at 0.0989 0.7683 RCIM_Constant>,<at 0.1984 1 RCIM_Constant>,<at 0.3016 1 RCIM_Constant>,<at 0.3996 0.6749 RCIM_Constant>,<at 0.5005 0.3398 RCIM_Constant>,<at 0.5977 1 RCIM_Constant>,<at 0.6957 1 RCIM_Constant>,<at 0.7982 1 RCIM_Constant>,<at 0.8984 0.2783 RCIM_Constant>,<at 1 1 RCIM_Constant>]}
       RedCurve {keys [<at 0 0.0496 RCIM_Constant>,<at 0.0989 0.1129 RCIM_Constant>,<at 0.1984 0.5464 RCIM_Constant>,<at 0.3016 0 RCIM_Constant>,<at 0.3996 0.0442 RCIM_Constant>,<at 0.5005 0 RCIM_Constant>,<at 0.5977 0 RCIM_Constant>,<at 0.6957 0 RCIM_Constant>,<at 0.7982 0.0496 RCIM_Constant>,<at 0.8984 1 RCIM_Constant>,<at 1 1 RCIM_Constant>]}
      }}
     CurveIndex=&RandomRangeFloat(Minimum=0, Maximum=1)
    )
    Lifetime Max=6
    Lifetime Min=3
    Lifetime Mode=::NewEnumerator1
    Mesh Scale Mode=::NewEnumerator1
    Mesh Uniform Scale Max=7
    Mesh Uniform Scale Min=5
    UsePositionOffset=false
    Write Material Random=true
   )
   &ShapeLocation(MaskSpawnGroup=false, Override Seed=false, Sphere Radius=7500, UseOwnerScale=true)
   /Niagara/Modules/Math/PlaceParticlesOnDepthBuffer_GPU(Max Z Depth=10000, Module.Use Max Z Depth=true)
   &KillParticles(Kill Particles=&Boolean_NotOperation(Bool=$Output.PlaceParticlesOnDepthBuffer_GPU.ValidSample)) {module=KillParticles001}
   &KillParticles(Kill Particles=&SetBoolByFloatComparison(
     A=&MakeFloatFromVector(Vector=Linked() {linkedVariable {name=Particles.Position, type=NiagaraPosition}}, Channel=::NewEnumerator2)
     B=&MakeFloatFromVector(
      Vector=&AddVectorToPosition(Position=/Niagara/DynamicInputs/Helpers/SimulationPosition(), Vector=<0 0 180>)
      Channel=::NewEnumerator2
     )
     Comparison Type=::NewEnumerator6
    )) {module=KillParticles002}
   &KillParticlesInVolume(Invert Volume=true, Sphere Radius=25000)
   /Game/Effects/Niagara/Modules/Spawn/FN_Spawn_GBuffer_Color_WithinRangeV2(HueMax=90, HueMin=86)
   &AddVelocity(Velocity=&RandomRangeVector(Minimum=<30 50 -250>, Maximum=<80 150 500>))
   &CurlNoiseForce(Bias Noise=false, Noise Frequency=25, Noise Strength=250, Pan Noise=false) {enabled=false}
   &ApplyInitialForces()
   &CameraOffset(Camera Offset Amount=-200) {enabled=false}
   &InitialMeshOrientation(Mesh Axis To Orient=<0 1 0>, Mesh Orientation Mode=::NewEnumerator2, Use Rotation Vector=false)
   &Color(Color=&RandomRangeLinearColor(Minimum=<rgba 0.0666 0.1144 0.0152 1>, Maximum=<rgba 0.0437 0.0742 0.0116 1>)) {enabled=false}
  ]
  particleUpdateScript [
   &ParticleState()
   &ScaleMeshSize(Scale Factor=&VectorFromFloat(Value=&FloatFromCurve(FloatCurve=DataInterface() {propertyValues {Curve {
         defaultValue=1
         keys [
          {interpMode=RCIM_Cubic, tangentMode=RCTM_User, time=0, value=0}
          {arriveTangent=1.25, interpMode=RCIM_Cubic, leaveTangent=0.0663, tangentMode=RCTM_Break, time=0.1, value=1}
          {interpMode=RCIM_Cubic, tangentMode=RCTM_User, time=0.8, value=1}
          {tangentMode=RCTM_User, time=1, value=0}
         ]
        }}})))
   /SetVariables(Particles.ScaleGnomon:Vector3f=&VectorFromFloat(Value=9)) {isSetParameters=true}
   &GravityForce(Gravity=&Multiply_VectorByFloat(Vector=<0 0 -980>, Float=&RandomRangeFloat(Minimum=1, Maximum=2))) {enabled=false}
   &WindForce(
    Friction=2
    Ground Distance Min / Max:Vector2f=<0 100>
    Ground Elevation:NiagaraFloat=&MakeFloatFromVector(Vector=Linked() {linkedVariable {name=Engine.Emitter.SimulationPosition, type=NiagaraPosition}}, Channel=::NewEnumerator2)
    Offset By Time:Vector3f=<0 0 100>
    Scale Min / Max:Vector2f=<0.5 1>
    Speed Min / Max:Vector2f=<0 150>
    Turbulence Scale Mode:ENiagaraWindTurbulenceContributionMode=::NewEnumerator3
    Use Ground Mask=true
    Wind Speed=<999 0 100>
    Wind Speed Scale=&Lerp_Float(A=0.25, B=1, Alpha=$NPC.FortniteNPC.WindGustsSystem_Active)
   ) {module=WindForce001}
   &AerodynamicDrag(
    Aerodynamic Drag:NiagaraFloat=&FloatFromCurve(
     FloatCurve=DataInterface() {propertyValues {Curve {
        defaultValue=1
        keys [{interpMode=RCIM_Cubic, tangentMode=RCTM_User, time=0, value=0}, {interpMode=RCIM_Cubic, tangentMode=RCTM_User, time=0.5, value=1}]
       }}}
     Scale Curve=0.5
    )
    Aerodynamic Rotational Drag=5
    Drag Shape Curve=DataInterface() {propertyValues {Curve {keys [
        {interpMode=RCIM_Cubic, tangentMode=RCTM_User, time=0, value=0.1}
        {arriveTangent=2.0799, interpMode=RCIM_Cubic, leaveTangent=2.0799, tangentMode=RCTM_User, time=1, value=1}
       ]}}}
    Lift Shape Curve=DataInterface() {propertyValues {Curve {keys [
        {interpMode=RCIM_Cubic, tangentMode=RCTM_User, time=0, value=0}
        {interpMode=RCIM_Cubic, tangentMode=RCTM_User, time=0.5, value=1}
        {tangentMode=RCTM_User, time=1, value=0}
       ]}}}
    Use Linear Drag=true
    Use Rotational Drag=true
   )
   &AerodynamicDrag(
    Aerodynamic Drag=0.5
    Drag Shape Curve=DataInterface() {propertyValues {Curve {keys [
        {interpMode=RCIM_Cubic, tangentMode=RCTM_User, time=0, value=0.1}
        {arriveTangent=2.0799, interpMode=RCIM_Cubic, leaveTangent=2.0799, tangentMode=RCTM_User, time=1, value=1}
       ]}}}
    Lift Shape Curve=DataInterface() {propertyValues {Curve {keys [
        {interpMode=RCIM_Cubic, tangentMode=RCTM_User, time=0, value=0}
        {interpMode=RCIM_Cubic, tangentMode=RCTM_User, time=0.5, value=1}
        {tangentMode=RCTM_User, time=1, value=0}
       ]}}}
    Use Linear Drag=true
    Use Rotational Drag=true
   ) {enabled=false, module=AerodynamicDrag001}
   &SolveForcesAndVelocity(Clamp Velocity=false, Limit Acceleration=false)
   /SetVariables(Particles.IsWind:NiagaraBool=&SetBoolByFloatComparison(A=$NPC.FortniteNPC.WindGustsSystem_Active, B=0.1, Comparison Type=::NewEnumerator2)) {isSetParameters=true}
   &DynamicMaterialParameters(
    Index 0 Param 1=&MakeCustomFloatFromBool(
     Bool=$Particles.IsWind
     True Float=&RandomRangeFloat(Minimum=1.1, Maximum=2)
     False Float=&Multiply_Float(
      A=&Waveform(
       "Waveform [1]"=::NewEnumerator0
       "[1] Frequency"=&RandomRangeFloat(Minimum=0.25, Maximum=1)
       "[1] Amplitude Min/Max"=<0.5 1>
      )
      B=&RandomRangeFloat(Minimum=0.4, Maximum=0.8)
     )
    )
    Index 0 Param 2=&Multiply_Float(
     A=&MakeCustomFloatFromBool(
      Bool=$Particles.IsWind
      True Float=&RandomRangeFloat(Minimum=1, Maximum=2)
      False Float=&RandomRangeFloat(Minimum=0.5, Maximum=1)
     )
     B=&RandomRangeFloat(Minimum=2, Maximum=3)
    )
    Index 0 Param 3=&RandomRangeFloat(Minimum=1, Maximum=2)
    Index 1 Param 1=&Lerp_Float(
     A=1
     B=2
     Alpha=&FloatFromCurve(
      FloatCurve=DataInterface() {propertyValues {Curve {
         defaultValue=1
         keys [
          {interpMode=RCIM_Cubic, tangentMode=RCTM_User, time=0, value=1}
          {interpMode=RCIM_Cubic, tangentMode=RCTM_User, time=0.3, value=0}
          {arriveTangent=0.2184, interpMode=RCIM_Cubic, leaveTangent=1.4286, tangentMode=RCTM_Break, time=0.7, value=0}
          {tangentMode=RCTM_User, time=1, value=1}
         ]
        }}}
      CurveIndex=Linked() {linkedVariable {name=NPC.FortniteNPC.TimeOfDay_0-1, type=NiagaraFloat}}
     )
    )
    Module.Index0Float0Write=true
    Module.Index0Float1Write=true
    Module.Index0Float2Write=true
    Module.Index0Float3Write=true
    Module.Index1Float0Write=true
    Module.Index1Float1Write=false
    Module.Index1Float2Write=true
    Module.Index1Float3Write=true
    Write Parameter Index 1=::NewEnumerator1
   ) {module=DynamicMaterialParameters001}
   &ScaleColor(Scale Alpha=&RandomRangeFloat(Minimum=0.7, Maximum=0.85), ScaleA=true, ScaleRGB=false)
   &CameraOffset(Camera Offset Amount=&UniformAOrBFloat(
     A=&FloatFromCurve(
      FloatCurve=DataInterface() {propertyValues {Curve {
         defaultValue=1
         keys [{interpMode=RCIM_Cubic, tangentMode=RCTM_User, time=0.4, value=0}, {interpMode=RCIM_Cubic, tangentMode=RCTM_User, time=1, value=1}]
        }}}
      Scale Curve=200
     )
     B=200
     Distribution Weight=0.75
     Evaluation Type=::NewEnumerator0
     Recalculate Random Each Loop=true
     Override Seed=false
    )) {module=CameraOffset001}
   &KillParticlesInVolume(
    Box Size:Vector3f=<1600 1800 1300>
    Kill Shape=::NewEnumerator1
    Volume Origin:Vector3f=&AddVectorToPosition(Position=<-34872 -50920 2480>, Vector=<0 0 0>)
   ) {module=KillParticlesInVolume001}
  ]
  rendererClasses [&NiagaraMeshRendererProperties]
  renderers [
   {rendererClass=&NiagaraMeshRendererProperties, rendererIndex=0}
   {rendererClass=&NiagaraMeshRendererProperties, rendererIndex=1}
  ]
 }
 Butterflies_OnGroundInteractAndIdleSpawn {
  simTarget=GPUComputeSim
  emitterUpdateScript [
   /SetVariables(
    Emitter.CameraForward:Vector3f=&ReturnCameraVector(Camera Property=::NewEnumerator1, Coordinate Space=::Simulation, CameraQuery=$Emitter.CameraRead)
    Emitter.CameraPosition:Vector3f=&ReturnCameraVector(Camera Property=::NewEnumerator0, Coordinate Space=::World, CameraQuery=$Emitter.CameraRead)
    Emitter.CameraRight:Vector3f=&ReturnCameraVector(Camera Property=::NewEnumerator3, Coordinate Space=::Simulation, CameraQuery=$Emitter.CameraRead)
    Emitter.CameraUp:Vector3f=&ReturnCameraVector(Camera Property=::NewEnumerator2, Coordinate Space=::Simulation, CameraQuery=$Emitter.CameraRead)
   ) {isSetParameters=true}
   &EmitterState(Life Cycle Mode=::NewEnumerator1, Loop Behavior=::NewEnumerator1, Loop Duration Mode=::NewEnumerator1, UseLoopDelay=false)
   &SpawnRate(SpawnGroup=1, SpawnRate=999, Use Spawn Probability=false) {module=SpawnRate001}
   &SpawnRate(SpawnRate=2000, Use Spawn Probability=false) {module=SpawnRate002}
  ]
  particleSpawnScript [
   &PartitionParticles(Number Of Partitions=3)
   &InitializeParticle(
    Color=&ColorFromCurve(
     ColorCurve=DataInterface() {propertyValues {
       AlphaCurve {keys [<at 0 1 RCIM_Constant>,<at 1 1 RCIM_Constant>]}
       BlueCurve {keys [<at 0 0.5881 RCIM_Constant>,<at 0.0989 1 RCIM_Constant>,<at 0.1984 0.0729 RCIM_Constant>,<at 0.3016 0.3289 RCIM_Constant>,<at 0.3996 1 RCIM_Constant>,<at 0.5005 1 RCIM_Constant>,<at 0.5977 0.5532 RCIM_Constant>,<at 0.6957 0.3262 RCIM_Constant>,<at 0.7982 0 RCIM_Constant>,<at 0.8984 0 RCIM_Constant>,<at 1 1 RCIM_Constant>]}
       GreenCurve {keys [<at 0 1 RCIM_Constant>,<at 0.0989 0.7683 RCIM_Constant>,<at 0.1984 1 RCIM_Constant>,<at 0.3016 1 RCIM_Constant>,<at 0.3996 0.6749 RCIM_Constant>,<at 0.5005 0.3398 RCIM_Constant>,<at 0.5977 1 RCIM_Constant>,<at 0.6957 1 RCIM_Constant>,<at 0.7982 1 RCIM_Constant>,<at 0.8984 0.2783 RCIM_Constant>,<at 1 1 RCIM_Constant>]}
       RedCurve {keys [<at 0 0.0496 RCIM_Constant>,<at 0.0989 0.1129 RCIM_Constant>,<at 0.1984 0.5464 RCIM_Constant>,<at 0.3016 0 RCIM_Constant>,<at 0.3996 0.0442 RCIM_Constant>,<at 0.5005 0 RCIM_Constant>,<at 0.5977 0 RCIM_Constant>,<at 0.6957 0 RCIM_Constant>,<at 0.7982 0.0496 RCIM_Constant>,<at 0.8984 1 RCIM_Constant>,<at 1 1 RCIM_Constant>]}
      }}
     CurveIndex=&RandomRangeFloat(Minimum=0, Maximum=1)
    )
    Lifetime=&RandomRangeFloat(Minimum=3, Maximum=5)
    Mesh Scale Mode=::NewEnumerator1
    Mesh Uniform Scale Max=10
    Mesh Uniform Scale Min=4
    Sprite Rotation Mode=::NewEnumerator2
    Sprite Size Mode=::NewEnumerator3
    Uniform Sprite Size=500
    UsePositionOffset=false
    Write Material Random=true
   )
   &ShapeLocation(MaskSpawnGroup=false, Offset Mode=::NewEnumerator0, Override Seed=false, Sphere Radius=5000, UseOwnerScale=true)
   /SetVariables(Particles.Position=&Add_Vector(
     A=Linked() {linkedVariable {name=Particles.Position, type=NiagaraPosition}}
     B=&MakeVector(X=0, Y=0, Z=&Waveform("Waveform [1]"=::NewEnumerator0, "[1] Frequency"=5, "[1] Amplitude Min/Max"=<0 25>))
    )) {isSetParameters=true}
   /SetVariables(Particles.New Vector 3f:Vector3f=<0 0 1>) {isSetParameters=true}
   &CameraOffset(Camera Offset Amount=15)
   /SetVariables(Particles.Is Active:NiagaraBool=/Niagara/DynamicInputs/Helpers/MaskBool(
     Mask Source:NiagaraInt32=0
     Mask Parameter:NiagaraInt32=?
     Mask:NiagaraInt32=1
     Masked Bool:NiagaraBool=true
     Passthrough Bool:NiagaraBool=false
    )) {isSetParameters=true}
   /SetVariables(Particles.InitialPos:NiagaraPosition=$Particles.Position) {isSetParameters=true}
   &PlaceParticlesOnDepthBuffer_GPU(Max Z Depth=10000, Module.Use Max Z Depth=true)
   &InitialMeshOrientation(
    Mesh Axis To Orient=<0 0 1>
    Mesh Orientation Mode=::NewEnumerator2
    Orientation Vector=$Output.PlaceParticlesOnDepthBuffer_GPU.DepthBufferSample_SampleWorldNormal
    Rotation=&RandomRangeVector(Minimum=<0 0 0>, Maximum=<0 0 1>)
    Rotation Coordinate Space=::NewEnumerator2
    Use Rotation Vector=true
   )
   &KillParticles(Kill Particles=&SetBoolByFloatComparison(
     A=&Divide_Float(A=&MakeFloatFromInt(Int=$Engine.Emitter.NumParticles), B=600)
     B=1
    )) {module=KillParticles003}
   /SetVariables(Particles.LimitCheck:NiagaraBool=&SetBoolByFloatComparison(
     A=&Divide_Float(A=&MakeFloatFromInt(Int=$Engine.Emitter.NumParticles), B=600)
     B=1
    )) {isSetParameters=true}
   &KillParticles(Kill Particles=&Boolean_NotOperation(Bool=$Output.PlaceParticlesOnDepthBuffer_GPU.ValidSample)) {module=KillParticles002}
   &FN_Spawn_GBuffer_Color_WithinRangeV2(
    HueMax:NiagaraFloat=&SelectFloatFromArray(Direct Array Index=$Particles.PartitionParticles.Partition, Float Selection Array=DataInterface() {propertyValues {FloatData [329,292,29]}})
    HueMin:NiagaraFloat=&SelectFloatFromArray(Direct Array Index=$Particles.PartitionParticles.Partition, Float Selection Array=DataInterface() {propertyValues {FloatData [325,287,32]}})
   )
   &AddVelocity(Velocity=<0 0 0>)
   &KillParticlesInVolume(Sphere Radius=500) {module=KillParticlesInVolume001}
   &KillParticlesInVolume(Invert Volume=true, Sphere Radius=5000)
   &Color(Color=&RandomRangeLinearColor(Minimum=<rgba 0.402 0.1221 0.2747 1>, Maximum=<rgba 0.1812 0.0578 0.1248 1>)) {enabled=false}
   &Color(Color=&RandomRangeLinearColor(Minimum=<rgba 0.3325 0.1248 0.3712 1>, Maximum=<rgba 0.1413 0.0545 0.159 1>)) {enabled=false, module=Color001}
   &Color(Color=&RandomRangeLinearColor(Minimum=<rgba 0.3968 0.2502 0.0999 1>, Maximum=<rgba 0.2789 0.1746 0.0704 1>)) {enabled=false, module=Color002}
  ]
  particleUpdateScript [
   &ParticleState(
    DeltaTime:NiagaraFloat=&MakeCustomFloatFromBool(
     Bool=&SetBoolByFloatComparison(A=&VectorLength(Vector=$Particles.Velocity), B=1)
     True Float=$Engine.DeltaTime
     False Float=0
    )
    Kill Particles When Lifetime Has Elapsed=false
    Let Infinitely Lived Particles Die When Emitter Deactivates=true
    Loop Particles Lifetime=true
   )
   /Niagara/Modules/Update/Forces/PointAttractionForce(
    Attraction Radius=1500
    AttractionStrength:NiagaraFloat=&Multiply_Float(
     A=&FloatFromCurve(
      FloatCurve=DataInterface() {propertyValues {Curve {keys [<at 0 1>,<at 0.25 1>]}}}
      Scale Curve=&RandomRangeFloat(Minimum=-3, Maximum=-5)
     )
     B=1
    )
    AttractorPosition:Vector3f=&Add_Vector(
     A=$Emitter.CameraPosition
     B=&Add_Vector(
      A=&Add_Vector(
       A=&Multiply_VectorByFloat(Vector=<0 0 1>, Float=-100)
       B=&Multiply_VectorByFloat(Vector=$Emitter.CameraForward, Float=-250)
      )
      B=&Multiply_VectorByFloat(Vector=$Emitter.CameraRight, Float=0)
     )
    )
    Falloff Exponent=1
    Module.Kill Within Radius=false
    Module.Use Falloff=false
   )
   /SetVariables(Particles.Normalized Velocity:Vector3f=$Particles.Velocity) {isSetParameters=true}
   /SetVariables(Particles.Is Active:NiagaraBool=&Boolean_OrOperation(
     Bool1=$Particles.Is Active
     Bool2=&SetBoolByFloatComparison(A=&VectorLength(Vector=$Particles.Normalized Velocity), B=1)
    )) {isSetParameters=true}
   /SetVariables(Particles.IsWind:NiagaraBool=&SetBoolByFloatComparison(A=$NPC.FortniteNPC.WindGustsSystem_Active, B=0.1, Comparison Type=::NewEnumerator2)) {isSetParameters=true}
   &CurlNoiseForce(
    Curl Noise Cone Mask Angle=90
    Mask Curl Noise=true
    Noise Frequency=&RandomRangeFloat(Minimum=0.85, Maximum=1)
    Noise Strength=&MakeCustomFloatFromBool(Bool=$Particles.Is Active, True Float=&RandomRangeFloat(Minimum=1500, Maximum=25000), False Float=0)
    Pan Noise=false
    Pan Noise Field=<0 0 0.1>
   )
   &GravityForce(Gravity=&Multiply_VectorByFloat(
     Vector=<0 0 1>
     Float=&MakeCustomFloatFromBool(
      Bool=$Particles.Is Active
      True Float=&MakeCustomFloatFromBool(
       Bool=$Particles.IsWind
       True Float=&RandomRangeFloat(Minimum=100, Maximum=500)
       False Float=&RandomRangeFloat(Minimum=500, Maximum=1500)
      )
      False Float=0
     )
    ))
   &PointAttractionForce(
    Attraction Radius=500
    AttractionStrength:NiagaraFloat=&Multiply_Float(
     A=&MakeCustomFloatFromBool(
      Bool=$Particles.Is Active
      True Float=0
      False Float=&FloatFromCurve(FloatCurve=DataInterface() {propertyValues {Curve {keys [<at 0 1>,<at 0.25 0>]}}}, Scale Curve=-8)
     )
     B=1
    )
    AttractorPosition:Vector3f=&self:ScratchDynamicInput(
     CameraPosition:NiagaraPosition=Linked() {linkedVariable {name=Emitter.CameraPosition, type=Vector3f}}
     CameraForward:Vector3f=$Emitter.CameraForward
     HeightOffset:NiagaraFloat=&RandomRangeFloat(Minimum=-100, Maximum=-150)
     CameraForwardOffset:NiagaraFloat=&RandomRangeFloat(Minimum=200, Maximum=250)
    )
    Falloff Exponent=1
    Module.Kill Within Radius=false
    Module.Use Falloff=false
   ) {module=PointAttractionForce001}
   &RegionMask(Inner Float=0, Outer Float=1, Region Radius=5000, UseOffset=false, UseRotation=false)
   &ScaleMeshSize(Scale Factor=&Multiply_Vector(
     A=&VectorFromFloat(Value=&FloatFromCurve(FloatCurve=DataInterface() {propertyValues {Curve {
          defaultValue=1
          keys [{interpMode=RCIM_Cubic, tangentMode=RCTM_User, time=0, value=0}, {interpMode=RCIM_Cubic, tangentMode=RCTM_User, time=0.2, value=1}]
         }}}))
     B=&VectorFromFloat(Value=&FloatFromCurve(
       FloatCurve=DataInterface() {propertyValues {Curve {
          defaultValue=1
          keys [{interpMode=RCIM_Cubic, tangentMode=RCTM_User, time=0.3, value=1}, {interpMode=RCIM_Cubic, tangentMode=RCTM_User, time=1, value=2}]
         }}}
       CurveIndex=$Particles.RegionMask.Float
      ))
    )) {module=ScaleMeshSize002}
   /Niagara/Modules/Math/AvoidDistanceFieldSurfaces_GPU(Avoid Nearest Surface=true, Avoid Oncoming Surfaces=true, Nearest Surface Avoidance Strength=500) {enabled=false}
   &SolveForcesAndVelocity(Clamp Velocity=true, Limit Acceleration=false, Speed Limit=400) {module=SolveForcesAndVelocity001}
   &CalculateAccurateVelocity()
   &ScaleColor(
    Scale Alpha=&Multiply_Float(
     A=&Multiply_Float(
      A=&Multiply_Float(
       A=1
       B=&FloatFromCurve(FloatCurve=DataInterface() {propertyValues {Curve {keys [<at 0 0>,<at 2 1>]}}}, CurveIndex=$Emitter.Age)
      )
      B=&FloatFromCurve(FloatCurve=DataInterface() {propertyValues {Curve {keys [
           {arriveTangent=9.0725, time=0, value=1}
           {arriveTangent=-0.4023, interpMode=RCIM_Cubic, leaveTangent=-0.4023, tangentMode=RCTM_User, time=0.75, value=0.85}
           {arriveTangent=-0.0405, interpMode=RCIM_Cubic, leaveTangent=-0.0405, tangentMode=RCTM_User, time=1, value=0}
          ]}}})
     )
     B=/Niagara/DynamicInputs/Scalability/ScalabilityDistanceBasedFloat(
      Curve:NiagaraDataInterfaceCurve=DataInterface() {propertyValues {Curve {keys [<at 0.7 1>,<at 1 0>]}}}
      CurveIndex:NiagaraFloat=$Emitter.Scalability.DistanceFraction
      Bias:NiagaraFloat=0
      Scale:NiagaraFloat=1
     )
    )
    ScaleA=true
    ScaleRGB=true
   ) {enabled=false}
   &ScaleColor(
    Scale Alpha=&Multiply_Float(
     A=&Multiply_Float(
      A=&Multiply_Float(
       A=1
       B=&FloatFromCurve(FloatCurve=DataInterface() {propertyValues {Curve {keys [<at 0 0>,<at 2 1>]}}}, CurveIndex=$Emitter.Age)
      )
      B=&FloatFromCurve(FloatCurve=DataInterface() {propertyValues {Curve {keys [
           {arriveTangent=9.0725, time=0, value=1}
           {arriveTangent=-0.4023, interpMode=RCIM_Cubic, leaveTangent=-0.4023, tangentMode=RCTM_User, time=0.75, value=0.85}
           {arriveTangent=-0.0405, interpMode=RCIM_Cubic, leaveTangent=-0.0405, tangentMode=RCTM_User, time=1, value=0}
          ]}}})
     )
     B=&ScalabilityDistanceBasedFloat(
      Curve:NiagaraDataInterfaceCurve=DataInterface() {propertyValues {Curve {keys [{interpMode=RCIM_Cubic, tangentMode=RCTM_User, time=0.7, value=1}, {interpMode=RCIM_Cubic, tangentMode=RCTM_User, time=1, value=0}]}}}
      CurveIndex:NiagaraFloat=$Emitter.Scalability.DistanceFraction
      Bias:NiagaraFloat=0
      Scale:NiagaraFloat=1
     )
    )
    ScaleA=true
    ScaleRGB=true
   ) {module=ScaleColor001}
   &DynamicMaterialParameters(
    Index 0 Param 1=&MakeCustomFloatFromBool(
     Bool=$Particles.Is Active
     True Float=&MakeCustomFloatFromBool(
      Bool=$Particles.IsWind
      True Float=&RandomRangeFloat(Minimum=2, Maximum=2.5)
      False Float=&RandomRangeFloat(Minimum=1.1, Maximum=2)
     )
     False Float=&Multiply_Float(
      A=&Waveform(
       "Waveform [1]"=::NewEnumerator0
       "[1] Frequency"=&RandomRangeFloat(Minimum=0.25, Maximum=1)
       "[1] Amplitude Min/Max"=<0.5 1>
      )
      B=&RandomRangeFloat(Minimum=0.4, Maximum=0.8)
     )
    )
    Index 0 Param 2=&Multiply_Float(
     A=&MakeCustomFloatFromBool(
      Bool=$Particles.Is Active
      True Float=&MakeCustomFloatFromBool(
       Bool=$Particles.IsWind
       True Float=&RandomRangeFloat(Minimum=2, Maximum=2.5)
       False Float=&RandomRangeFloat(Minimum=1, Maximum=2)
      )
      False Float=&RandomRangeFloat(Minimum=0.5, Maximum=1)
     )
     B=&RandomRangeFloat(Minimum=2, Maximum=3)
    )
    Index 0 Param 3=&RandomRangeFloat(Minimum=1, Maximum=2)
    Index 1 Param 1=&FloatFromCurve(FloatCurve=DataInterface() {propertyValues {Curve {
        defaultValue=1
        keys [
         {interpMode=RCIM_Cubic, tangentMode=RCTM_User, time=0, value=3}
         {interpMode=RCIM_Cubic, tangentMode=RCTM_User, time=0.5, value=0.5}
         {tangentMode=RCTM_User, time=1, value=3}
        ]
       }}})
    Index 1 Param 2=&RandomRangeFloat(Minimum=0, Maximum=0.15)
    Module.Param0WriteEnabled=true
    Module.Param0WriteEnabled1=true
    Module.Param1WriteEnabled=true
    Module.Param1WriteEnabled1=true
    Module.Param2WriteEnabled=true
    Module.Param2WriteEnabled1=true
    Module.Param3WriteEnabled=true
    Module.Param3WriteEnabled1=true
    Write Parameter Index 1=true
   )
   &CameraOffset(Camera Offset Amount=&Lerp_Float(
     A=10
     B=222
     Alpha=&MakeCustomFloatFromBool(
      Bool=$Particles.Is Active
      True Float=&FloatFromCurve(FloatCurve=DataInterface() {propertyValues {Curve {
          defaultValue=1
          keys [{interpMode=RCIM_Cubic, tangentMode=RCTM_User, time=0, value=1}, {interpMode=RCIM_Cubic, tangentMode=RCTM_User, time=0.5, value=0}]
         }}})
      False Float=1
     )
    )) {module=CameraOffset002}
   &ScaleMeshSize(Scale Factor=&MakeVector(
     X=1
     Y=1
     Z=&FloatFromCurve(
      FloatCurve=DataInterface() {propertyValues {Curve {keys [<at 0 1>,<at 0.22 2>,<at 0.25 1>]}}}
      Scale Curve=&Waveform("Waveform [1]"=::NewEnumerator0, "[1] Frequency"=3, "[1] Amplitude Min/Max"=<1 1.5>)
     )
    )) {enabled=false}
   /SetVariables(Particles.test:Vector2f=&Vector2DFromFloat(Value=&Multiply_Float(
      A=55
      B=&FloatFromCurve(
       FloatCurve=DataInterface() {propertyValues {Curve {
          defaultValue=9
          keys [{interpMode=RCIM_Cubic, tangentMode=RCTM_User, time=0, value=1}, {interpMode=RCIM_Cubic, tangentMode=RCTM_User, time=1, value=9}]
         }}}
       CurveIndex=$Particles.RegionMask.Float
      )
     ))) {enabled=false, isSetParameters=true}
   &self:ScratchModule_02(ZDampening=1) {module=ScratchModule_02}
  ]
  rendererClasses [&NiagaraMeshRendererProperties, &NiagaraSpriteRendererProperties]
  renderers [
   {rendererClass=&NiagaraMeshRendererProperties, rendererIndex=0}
   {rendererClass=&NiagaraSpriteRendererProperties, rendererIndex=1}
  ]
 }
 Firefiles_Epic {
  simTarget=GPUComputeSim
  emitterSpawnScript [/SetVariables(Emitter.ScaleMult=1) {isSetParameters=true}]
  emitterUpdateScript [
   &EmitterState()
   &SpawnRate(
    SpawnRate=&FloatFromCurve(
     FloatCurve=DataInterface() {propertyValues {Curve {
        defaultValue=1
        keys [
         {interpMode=RCIM_Cubic, tangentMode=RCTM_User, time=0, value=1}
         {interpMode=RCIM_Cubic, tangentMode=RCTM_Break, time=0.3, value=0}
         {arriveTangent=0.4792, interpMode=RCIM_Cubic, leaveTangent=0.4792, tangentMode=RCTM_User, time=0.7, value=0}
         {tangentMode=RCTM_User, time=1, value=1}
        ]
       }}}
     CurveIndex=Linked() {linkedVariable {name=NPC.FortniteNPC.TimeOfDay_0-1, type=NiagaraFloat}}
     Scale Curve=&RandomRangeFloat(Minimum=12, Maximum=20)
    )
    Use Spawn Probability=false
   ) {module=SpawnRate002}
   &SpawnRate(SpawnRate=99, Use Spawn Probability=false) {enabled=false, module=SpawnRate001}
  ]
  particleSpawnScript [
   &InitializeParticle(
    Color=&ColorFromCurve(ColorCurve=DataInterface() {propertyValues {
       AlphaCurve {keys [<at 0 1 RCIM_Cubic>]}
       BlueCurve {keys [<at 0 0 RCIM_Cubic>,<at 0.2619 0>,<at 1 0 RCIM_Cubic>]}
       GreenCurve {keys [<at 0 0.3419 RCIM_Cubic>,<at 0.2619 0.1039>,<at 1 0.1039 RCIM_Cubic>]}
       RedCurve {keys [<at 0 1 RCIM_Cubic>,<at 0.2619 1>,<at 1 1 RCIM_Cubic>]}
      }})
    Lifetime=5
    Mesh Scale Mode=::NewEnumerator1
    Mesh Uniform Scale Max=0.6
    Mesh Uniform Scale Min=0.36
    Sprite Size Mode=::NewEnumerator3
    Uniform Sprite Size=75
    UsePositionOffset=false
    Write Material Random=true
   ) {enabled=false}
   &InitializeParticle(
    Color Maximum=<rgba 0.6222 1 0 1>
    Color Minimum=<rgba 0.3394 1 0 1>
    Color Mode=::NewEnumerator2
    Lifetime=5
    Mesh Scale Mode=::NewEnumerator3
    Mesh Uniform Scale=&Multiply_Float(A=&RandomRangeFloat(Minimum=0.36, Maximum=0.6), B=$Emitter.ScaleMult)
    Sprite Size Mode=::NewEnumerator3
    Uniform Sprite Size=&Multiply_Float(A=75, B=$Emitter.ScaleMult)
    UsePositionOffset=false
    Write Material Random=true
   ) {module=InitializeParticle001}
   &ShapeLocation(MaskSpawnGroup=false, Override Seed=false, Sphere Radius=7500, UseOwnerScale=true)
   &self:SampleLandscape_02() {module=SampleLandscape_02}
   &KillParticlesInVolume(Invert Volume=true, Sphere Radius=12500)
   &CameraOffset(Camera Offset Amount=1)
   &AccelerationForce(Acceleration=&Multiply_VectorByFloat(Vector=$Output.SampleLandscape_02.LandscapeNormal, Float=300)) {module=AccelerationForce001}
   &ApplyInitialForces()
   &FN_Spawn_GBuffer_Color_WithinRangeV2(HueMax=150, HueMin=138)
   /SetVariables(Particles.Position=&AddVectorToPosition(
     Position=$StackContext.Position
     Vector=&Multiply_Vector(A=$Output.SampleLandscape_02.LandscapeNormal, B=<1 1 200>)
    )) {enabled=false, isSetParameters=true}
   &Color(Color=&RandomRangeLinearColor(Minimum=<rgba 0.0097 0.1195 0.063 1>, Maximum=<rgba 0.0273 0.1413 0.0648 1>)) {enabled=false}
  ]
  particleUpdateScript [
   &ParticleState()
   /SetVariables(Particles.Jitter:NiagaraFloat=&Waveform(
     "Waveform [1]"=::NewEnumerator12
     "[1] Frequency"=&RandomRangeFloat(Minimum=15, Maximum=22)
     "[1] Amplitude Min/Max"=&RandomRangeVector2D(Minimum=<0 1>, Maximum=<0.3 1>)
     "[1] Scale"=3
    )) {isSetParameters=true}
   &ScaleMeshSize(Scale Factor=&VectorFromFloat(Value=1)) {module=ScaleMeshSize001}
   &ScaleMeshSize(Scale Factor=&VectorFromFloat(Value=&FloatFromCurve(FloatCurve=DataInterface() {propertyValues {Curve {keys [
          <at 0 0 RCIM_Cubic>
          {arriveTangent=1.1765, leaveTangent=1.1765, time=0.1, value=1}
          {interpMode=RCIM_Cubic, tangentMode=RCTM_Break, time=0.8, value=1}
          {arriveTangent=-23.4104, interpMode=RCIM_Cubic, tangentMode=RCTM_Break, time=1, value=0}
         ]}}})))
   &ScaleSpriteSize(Uniform Curve Sprite Scale=DataInterface() {propertyValues {Curve {keys [
        <at 0 0 RCIM_Cubic>
        {arriveTangent=1.1765, leaveTangent=1.1765, time=0.1, value=1}
        {interpMode=RCIM_Cubic, tangentMode=RCTM_Break, time=0.8, value=1}
        {arriveTangent=-23.4104, interpMode=RCIM_Cubic, tangentMode=RCTM_Break, time=1, value=0}
       ]}}})
   &CurlNoiseForce(Bias Noise=false, Noise Frequency=5, Noise Strength=750, Pan Noise=false)
   &WindForce(Wind Speed Scale=&Lerp_Float(A=0.1, B=5, Alpha=$NPC.FortniteNPC.WindGustsSystem_Active))
   &Drag(Drag=10, Use Linear Drag=true, Use Rotational Drag=true)
   &AccelerationForce(Acceleration=<50 0 300>)
   /Niagara/Modules/Update/Forces/V2/PointAttractionForce(Attraction Radius=150, Attraction Strength=-500, Kill Within Radius=false, Module.Use Falloff=true, UseAttractorOffset=false)
   &AvoidDistanceFieldSurfaces_GPU(Avoid Nearest Surface=true, Avoid Oncoming Surfaces=true, Nearest Surface Avoidance Strength=5)
   &SolveForcesAndVelocity(Clamp Velocity=false, Limit Acceleration=false)
   &DynamicMaterialParameters(
    Index 0 Param 4=&Lerp_Float(A=10, B=25, Alpha=&Sine(Normalized Angle=$Particles.Age, Period=1, Scale=1, Bias=0.5))
    Index 1 Param 1=1
    Module.Index0Param0Write=true
    Module.Index0Param1Write=true
    Module.Index0Param2Write=true
    Module.Index0Param3Write=true
    Module.Index1Param0Write=true
    Module.Index1Param1Write=true
    Module.Index1Param2Write=true
    Module.Index1Param3Write=true
    Write Parameter Index 1=true
   )
   &ScaleColor(
    Curve Index=&Multiply_Float(
     A=$Particles.NormalizedAge
     B=&FloatFromCurve(
      FloatCurve=DataInterface() {propertyValues {Curve {
         defaultValue=0.1
         keys [
          {interpMode=RCIM_Cubic, tangentMode=RCTM_User, time=0, value=1}
          {arriveTangent=0.1559, interpMode=RCIM_Cubic, leaveTangent=0.1559, tangentMode=RCTM_User, time=0.2, value=0}
          {interpMode=RCIM_Cubic, tangentMode=RCTM_User, time=0.8, value=0}
          {tangentMode=RCTM_User, time=1, value=1}
         ]
        }}}
      CurveIndex=Linked() {linkedVariable {name=NPC.FortniteNPC.TimeOfDay_0-1, type=NiagaraFloat}}
     )
    )
    Linear Color Curve=DataInterface() {propertyValues {AlphaCurve {keys [<at 0.8906 0.45>,<at 1 0>]}, BlueCurve {keys [<at 0 0.5>]}, GreenCurve {keys [<at 0 0.5>]}, RedCurve {keys [<at 0 0.5>]}}}
    Scale Mode=::NewEnumerator2
   )
   &InterpolateOverTime(Rate Of Change=5, Target Position Value:NiagaraPosition=$Particles.Position, Variable Type=::NewEnumerator6)
   /SetVariables(Particles.Vector 4:Vector4f=&MakeVector4(
     X=&MakeFloatFromVector(
      Vector=&Subtract_Position(A=$Particles.InterpolateOverTime.InterpolateOverTime_Position.Moving Average, B=$Particles.Position)
      Channel=::NewEnumerator0
     )
     Y=&MakeFloatFromVector(
      Vector=&Subtract_Position(A=$Particles.InterpolateOverTime.InterpolateOverTime_Position.Moving Average, B=$Particles.Position)
      Channel=::NewEnumerator1
     )
     Z=&MakeFloatFromVector(
      Vector=&Subtract_Position(A=$Particles.InterpolateOverTime.InterpolateOverTime_Position.Moving Average, B=$Particles.Position)
      Channel=::NewEnumerator2
     )
     W=1
    )) {isSetParameters=true}
   &ScaleMeshSize(Scale Factor=&VectorFromFloat(Value=&Lerp_Float(
      A=0
      B=1
      Alpha=&RemapRange(
       Input Value To Remap=&VectorLength(Vector=&SubtractVectorFromPosition(
         Position=&ReturnCameraPosition(Coordinate Space:ENiagaraCoordinateSpace=::Simulation, Camera Query:NiagaraDataInterfaceCamera=DataInterface())
         Vector=<0 0 0>
        ))
       Input Min=0
       Input Max=1
       Output Min=0
       Output Max=1
      )
     ))) {module=ScaleMeshSize002}
   &FN_Sprite_Attributes(
    Camera Offset=20
    Color=&Multiply_LinearColorByFloat(
     Linear Color=&HueShiftLinearColor(Hue=0, Saturation=0.25, Value=&Multiply_Float(A=5, B=$Particles.Jitter), Linear Color=$Particles.Color)
     Float=&FloatFromCurve(FloatCurve=DataInterface() {propertyValues {Curve {keys [
          <at 0.05 0 RCIM_Cubic>
          {arriveTangent=1.1765, time=0.2, value=1}
          {interpMode=RCIM_Cubic, tangentMode=RCTM_Break, time=0.8, value=1}
          {arriveTangent=-13.7968, interpMode=RCIM_Cubic, tangentMode=RCTM_Break, time=1, value=0}
         ]}}})
    )
    Create Camera Offset Attribute=true
    Create Color Attribute=true
    Create Position Attribute=true
    Create SpriteSize Attribute=true
    Position:NiagaraPosition=&AddVectorToPosition(
     Position=$Particles.Position
     Vector=&TransformVector(Vector=<0 0 0>, Source Space=::World, Destination Space=::Local, Apply Scale=true)
    )
    SpriteSize=&Multiply_Vector2D_ByFloat(
     Float=&Lerp_Float(
      A=0
      B=1
      Alpha=&RemapRange(
       Input Value To Remap=&VectorLength(Vector=&SubtractVectorFromPosition(
         Position=&SimulationPosition()
         Vector=Linked() {linkedVariable {name=Particles.Position, type=NiagaraPosition}}
        ))
       Input Min=0
       Input Max=2500
       Output Min=0
       Output Max=1
      )
     )
     Vector2D=&Multiply_Vector2D_ByFloat(Float=$Emitter.ScaleMult, Vector2D=<40 40>)
    )
    Write to Module Namespace=true
   )
   /SetVariables(Particles.Vector 4:Vector4f=<1 1 0 1>) {isSetParameters=true}
  ]
  rendererClasses [&NiagaraMeshRendererProperties, &NiagaraSpriteRendererProperties]
  renderers [
   {rendererClass=&NiagaraMeshRendererProperties, rendererIndex=0}
   {rendererClass=&NiagaraSpriteRendererProperties, rendererIndex=1}
   {rendererClass=&NiagaraSpriteRendererProperties, rendererIndex=2}
  ]
 }
 LeavesInAir {
  simTarget=GPUComputeSim
  emitterUpdateScript [
   /SetVariables(
    Emitter.Emitter_HeightOffset:NiagaraFloat=$System.System_HeightOffset
    Emitter.Overall_SpeedMult=0.2
    Emitter.SpawnRate_Curve:NiagaraFloat=&FloatFromCurve(
     FloatCurve=DataInterface() {propertyValues {Curve {keys [{arriveTangent=17.5837, time=0, value=0}, <at 0.1 1>, <at 0.2 0.0925>, <at 0.6 0.1525>, <at 0.7 0>, <at 1 0>]}}}
     CurveIndex=$Emitter.NormalizedLoopAge
     Scale Curve=100
    )
    Emitter.SpawnSpeed_Curve:NiagaraFloat=&FloatFromCurve(
     FloatCurve=DataInterface() {propertyValues {Curve {keys [
         {arriveTangent=30.2685, interpMode=RCIM_Cubic, leaveTangent=30.2685, tangentMode=RCTM_User, time=0.0009, value=0.0215}
         {arriveTangent=-0.0619, interpMode=RCIM_Cubic, leaveTangent=-0.0619, tangentMode=RCTM_User, time=0.0989, value=0.3515}
         <at 0.297 0.0032>
         <at 0.3192 0.3332>
         <at 0.3525 0.0398>
         <at 0.6171 0.3148>
         <at 0.9939 0.3698>
        ]}}}
     CurveIndex=&Multiply_Float(A=$Emitter.NormalizedLoopAge, B=&RandomRangeFloat(Minimum=1, Maximum=5))
     Scale Curve=&Multiply_Float(
      A=&Lerp_Float(A=5, B=10, Alpha=&UniformRangedFloat(Minimum=0.25, Maximum=1))
      B=1
     )
    )
    Emitter.SpriteRotRate:NiagaraFloat=&FloatFromCurve(
     FloatCurve=DataInterface() {propertyValues {Curve {keys [
         <at 0.0041 0.1315>
         <at 0.1088 1.9664>
         {arriveTangent=-3.3757, interpMode=RCIM_Cubic, leaveTangent=-3.3757, tangentMode=RCTM_User, time=0.3609, value=0.6391}
         {arriveTangent=-1, interpMode=RCIM_Cubic, leaveTangent=-1, tangentMode=RCTM_User, time=0.5614, value=0.4386}
         <at 0.7086 1.6472>
         <at 1 0>
        ]}}}
     CurveIndex=$Emitter.NormalizedLoopAge
     Scale Curve=500
    )
   ) {isSetParameters=true}
   &EmitterState(Life Cycle Mode=::NewEnumerator1, Loop Duration=0.05, UseLoopDelay=false)
   &SpawnBurst_Instantaneous(
    Spawn Count=&RandomRangeInt(Minimum=40, Maximum=200)
    Use Spawn Probability=false
    UseLoopCountLimit=false
   ) {enabled=false}
   &SpawnRate(
    SpawnRate=&Multiply_Float(A=1000, B=&Lerp_Float(A=0.4, B=1.5, Alpha=$NPC.FortniteNPC.WindGustsSystem_Active))
    Use Spawn Probability=false
   )
  ]
  particleSpawnScript [
   /SetVariables(Particles.HeightOffset:NiagaraFloat=&RandomRangeFloat(Minimum=17500, Maximum=3000)) {isSetParameters=true}
   &PartitionParticles(Number Of Partitions=4) {enabled=false}
   &InitializeParticle(
    Color Maximum=<rgba 0.1265 0.2908 0.0132 1>
    Color Minimum=<rgba 0.061 0.256 0.0462 1>
    Color Mode=::NewEnumerator2
    Lifetime=&UniformRangedFloat(Minimum=3, Maximum=6)
    Mass Mode=::NewEnumerator1
    Mesh Scale Mode=::NewEnumerator1
    Mesh Uniform Scale Max=1.7
    Mesh Uniform Scale Min=1.4
    Sprite Rotation Angle=&UniformRangedFloat(Minimum=-180, Maximum=180)
    Sprite Rotation Mode=::NewEnumerator0
    Sprite Size=&Vector2DFromFloat(Value=&UniformRangedFloat(Minimum=100, Maximum=150))
    Sprite Size Mode=::NewEnumerator4
    UsePositionOffset=false
    Write Material Random=true
   ) {module=InitializeParticle001}
   &AddVelocity(Velocity=&Multiply_VectorByFloat(
     Vector=&Multiply_VectorByFloat(
      Vector=&Multiply_VectorByFloat(Vector=&UniformRangedVector(Minimum=<30 50 -250>, Maximum=<80 150 500>), Float=$Emitter.SpawnSpeed_Curve)
      Float=$Emitter.Overall_SpeedMult
     )
     Float=1
    ))
   /Niagara/Modules/Spawn/Location/SphereLocation(
    MaskSpawnGroup=false
    Module.Override Randomness=false
    Non Uniform Scale=&RandomRangeVector(Minimum=<-1 -1 -1>, Maximum=<1 1 1>)
    Offset=<0 0 900>
    Sphere Radius=&Multiply_Float(
     A=9500
     B=&FloatFromCurve(FloatCurve=DataInterface() {propertyValues {Curve {keys [<at 0 0.1>,<at 0.157 0.9198>]}}}, CurveIndex=$System.NormalizedLoopAge)
    )
    Surface Only=false
   ) {module=SphereLocation001}
   /SetVariables(Particles.InitialPos:NiagaraPosition=$Particles.Position) {isSetParameters=true}
   &PlaceParticlesOnDepthBuffer_GPU(Max Z Depth=10000, Module.Use Max Z Depth=true)
   &KillParticles(Kill Particles=&SetBoolByFloatComparison(
     A=&VectorLength(Vector=&Subtract_Vector(A=$Particles.InitialPos, B=Linked() {linkedVariable {name=Particles.Position, type=NiagaraPosition}}))
     B=1500
    ))
   &KillParticles(Kill Particles=&Boolean_NotOperation(Bool=$Output.PlaceParticlesOnDepthBuffer_GPU.ValidSample)) {module=KillParticles001}
   &KillParticles(Kill Particles=&SetBoolByFloatComparison(
     A=&MakeFloatFromVector(Vector=Linked() {linkedVariable {name=Particles.Position, type=NiagaraPosition}}, Channel=::NewEnumerator2)
     B=&MakeFloatFromVector(
      Vector=&AddVectorToPosition(Position=&SimulationPosition(), Vector=<0 0 180>)
      Channel=::NewEnumerator2
     )
     Comparison Type=::NewEnumerator6
    )) {module=KillParticles002}
   &KillParticlesInVolume(Invert Volume=true, Sphere Radius=12500)
   &FN_Spawn_GBuffer_Color_WithinRangeV2(HueMax=100, HueMin=93)
   /SetVariables(Particles.Init:NiagaraPosition=$Particles.Position) {isSetParameters=true}
   &CurlNoiseForce(
    Noise Frequency=25
    Noise Strength=&Multiply_Float(
     A=250
     B=&Multiply_Float(A=&Multiply_Float(A=1, B=$Emitter.Overall_SpeedMult), B=1)
    )
    Pan Noise=false
   ) {module=CurlNoiseForce001}
   &ApplyInitialForces()
   &CameraOffset(Camera Offset Amount=-200)
   &Color(Color=&RandomRangeLinearColor(Minimum=<rgba 0.0578 0.1329 0.006 1>, Maximum=<rgba 0.0222 0.0931 0.0168 1>)) {enabled=false}
  ]
  particleUpdateScript [
   /SetVariables(Particles.ParticleUpdate_RotRate:NiagaraFloat=&FloatFromCurve(
     FloatCurve=DataInterface() {propertyValues {Curve {keys [
         <at 0.0041 0.1315>
         <at 0.1088 1.9664>
         {arriveTangent=-3.3757, interpMode=RCIM_Cubic, leaveTangent=-3.3757, tangentMode=RCTM_User, time=0.3609, value=0.6391}
         {arriveTangent=-1, interpMode=RCIM_Cubic, leaveTangent=-1, tangentMode=RCTM_User, time=0.5614, value=0.4386}
         <at 0.7086 1.6472>
         {arriveTangent=-1.004, interpMode=RCIM_Cubic, leaveTangent=-1.004, tangentMode=RCTM_User, time=0.8311, value=0.9545}
         {arriveTangent=-5.6531, interpMode=RCIM_Cubic, leaveTangent=-5.6531, tangentMode=RCTM_User, time=0.8987, value=1.4563}
         <at 0.9384 0.3483>
         <at 1 0>
        ]}}}
     Scale Curve=5
    )) {isSetParameters=true}
   &ParticleState()
   &UpdateMeshOrientation(Rotation Rate=&FloatFromCurve(
     FloatCurve=DataInterface() {propertyValues {Curve {keys [<at 250 1>,<at 500 4>]}}}
     CurveIndex=&MakeFloatFromVector(Vector=$Particles.Velocity, Channel=::NewEnumerator0)
    ))
   &ScaleMeshSize(Scale Factor=&Multiply_VectorByFloat(
     Vector=<1 1 1>
     Float=&FloatFromCurve(FloatCurve=DataInterface() {propertyValues {Curve {keys [
          {interpMode=RCIM_Cubic, tangentMode=RCTM_User, time=0, value=0}
          {arriveTangent=2, leaveTangent=2, time=0.1, value=1}
          <at 0.5 1>
          {arriveTangent=-2, leaveTangent=-2, time=0.8, value=1}
          {tangentMode=RCTM_User, time=1, value=0}
         ]}}})
    ))
   /Niagara/Modules/Update/SubUV/SubUVAnimation(Number Of Frames=64)
   &ScaleSpriteSize(Scale Factor=&Multiply_Vector2(
     A=<1 1>
     B=&Vector2DFromFloat(Value=&Multiply_Float(
       A=&FloatFromCurve(FloatCurve=DataInterface() {propertyValues {Curve {keys [
            {interpMode=RCIM_Cubic, tangentMode=RCTM_User, time=-0.0061, value=0.489}
            {arriveTangent=0.3677, interpMode=RCIM_Cubic, leaveTangent=0.3677, tangentMode=RCTM_User, time=0.2892, value=0.9107}
            {arriveTangent=-1.2812, interpMode=RCIM_Cubic, leaveTangent=-1.2812, tangentMode=RCTM_User, time=0.9809, value=0.9275}
            <at 1 0>
           ]}}})
       B=&FloatFromCurve(
        FloatCurve=DataInterface() {propertyValues {Curve {keys [
            {interpMode=RCIM_Cubic, tangentMode=RCTM_User, time=0, value=0.0005}
            {arriveTangent=0.0074, interpMode=RCIM_Cubic, leaveTangent=0.0074, tangentMode=RCTM_User, time=0.15, value=1.0236}
            {arriveTangent=-0.0952, interpMode=RCIM_Cubic, leaveTangent=-0.0952, tangentMode=RCTM_User, time=0.85, value=1}
            <at 0.9 0>
           ]}}}
        CurveIndex=&NormalizeDistanceRange(
         Start Position:NiagaraPosition=Linked() {linkedVariable {name=Emitter.CameraPostion, type=Vector3f}}
         End Position:NiagaraPosition=$Particles.Position
         Distance:NiagaraFloat=10000
         Invert Normalized Range:NiagaraBool=false
        )
        Scale Curve=0.85
       )
      ))
    )) {module=ScaleSpriteSize001}
   &Drag(Use Linear Drag=true, Use Rotational Drag=false)
   &CurlNoiseForce(
    Curl Noise Cone Mask Angle=50
    Mask Curl Noise=true
    Noise Frequency=&RandomRangeFloat(Minimum=5, Maximum=25)
    Noise Strength=&Multiply_Float(A=$Emitter.SpawnSpeed_Curve, B=500)
    Pan Noise=true
    Pan Noise Field=<0 0 1>
    Random Seed=80
   )
   &ScaleColor(
    Scale Alpha=&FloatFromCurve(
     FloatCurve=DataInterface() {propertyValues {Curve {keys [
         {interpMode=RCIM_Cubic, tangentMode=RCTM_User, time=0, value=0}
         {arriveTangent=-0.0948, interpMode=RCIM_Cubic, leaveTangent=-0.0948, tangentMode=RCTM_User, time=0.1, value=1}
         <at 0.942 0.9825>
         <at 1 0>
        ]}}}
     Scale Curve=&FloatFromCurve(FloatCurve=DataInterface() {propertyValues {Curve {keys [
          {interpMode=RCIM_Cubic, tangentMode=RCTM_User, time=0, value=0}
          {interpMode=RCIM_Cubic, tangentMode=RCTM_User, time=0.2, value=5}
          {tangentMode=RCTM_User, time=1, value=0}
         ]}}})
    )
    ScaleA=true
    ScaleRGB=true
   )
   /Niagara/Modules/Update/SubUV/V2/SubUVAnimation(
    Start Frame:NiagaraInt32=&RandomRangeInt(Minimum=0, Maximum=32)
    SubUV Animation Mode:ENiagara_SubUVLookupMode=::NewEnumerator1
    SubUV Loop Count:NiagaraFloat=&SelectFloatFromArray(
     Array Sampling Mode=::NewEnumerator0
     Direct Array Index=?
     Float Selection Array=DataInterface() {propertyValues {FloatData [2,4]}}
     Random Seed=0
    )
   ) {module=SubUVAnimation001}
   &DynamicMaterialParameters(
    Index 0 Param 3=1
    Module.Param0WriteEnabled=true
    Module.Param1WriteEnabled=true
    Module.Param2WriteEnabled=true
    Module.Param3WriteEnabled=true
   )
   &GravityForce(Gravity=&Multiply_VectorByFloat(Vector=<0 0 -980>, Float=&RandomRangeFloat(Minimum=1, Maximum=2)))
   &WindForce(
    Friction=2
    Ground Distance Min / Max:Vector2f=<0 100>
    Ground Elevation:NiagaraFloat=&MakeFloatFromVector(Vector=Linked() {linkedVariable {name=Engine.Emitter.SimulationPosition, type=NiagaraPosition}}, Channel=::NewEnumerator2)
    Scale Min / Max:Vector2f=<0.5 1>
    Speed Min / Max:Vector2f=<0 150>
    Turbulence Scale Mode:ENiagaraWindTurbulenceContributionMode=::NewEnumerator3
    Use Ground Mask=true
    Wind Speed=<250 0 -100>
    Wind Speed Scale=&Lerp_Float(A=0.25, B=1, Alpha=$NPC.FortniteNPC.WindGustsSystem_Active)
   ) {module=WindForce001}
   &AerodynamicDrag(Aerodynamic Drag=0.5, Use Linear Drag=true, Use Rotational Drag=true)
   &SolveForcesAndVelocity(Clamp Velocity=false, Limit Acceleration=false)
   &AlignSpriteToMeshOrientation(Mesh Orientation Relative Sprite Facing Vector:Vector3f=<0 0 0>, SetAlignment=true, SetFacing=true)
  ]
  rendererClasses [&NiagaraMeshRendererProperties, &NiagaraSpriteRendererProperties]
  renderers [
   {rendererClass=&NiagaraMeshRendererProperties, rendererIndex=0}
   {rendererClass=&NiagaraSpriteRendererProperties, rendererIndex=1}
   {rendererClass=&NiagaraMeshRendererProperties, rendererIndex=2}
  ]
 }
}
summary {
 systemName=NS_BugBiome_BaseFX_Butterflies
 emitters [
  {
   emitterName=Firefiles_Epic
   simTarget=GPUComputeSim
   rendererClasses [&NiagaraMeshRendererProperties, &NiagaraSpriteRendererProperties]
  }
  {
   emitterName=LeavesInAir
   simTarget=GPUComputeSim
   rendererClasses [&NiagaraMeshRendererProperties, &NiagaraSpriteRendererProperties]
  }
  {bEnabled=false, emitterName=Butterflies_InTrees, simTarget=GPUComputeSim, rendererClasses [&NiagaraMeshRendererProperties]}
  {
   emitterName=Butterflies_OnGroundInteractAndIdleSpawn
   simTarget=GPUComputeSim
   rendererClasses [&NiagaraMeshRendererProperties, &NiagaraSpriteRendererProperties]
  }
 ]
 userVariables [User.Landscape {
   type=NiagaraDataInterfaceLandscape
   description=""
   defaultValue {
    struct=NiagaraExt_VariableValue_DataInterface
    value {
     dataInterface=&self:NiagaraDataInterfaceLandscape_0
     dataInterfaceClass=&NiagaraDataInterfaceLandscape
    }
   }
  }]
}
